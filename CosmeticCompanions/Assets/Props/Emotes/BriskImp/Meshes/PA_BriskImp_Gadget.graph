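# Physics asset for the BriskImp Gadget mesh, an emote prop in CosmeticCompanions, providing its simplified collision shape.
# A single capsule body on the Main bone, slightly offset and rotated, approximates the gadget.
# It has no constraints, so the asset defines one body with no jointed parts.
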PhysicsAsset /CosmeticCompanions/Assets/Props/Emotes/BriskImp/Meshes/PA_BriskImp_Gadget
bodies {Main [{shapeType=Capsule, length=27.8503, radius=24.879, center=<0.0004 -0.9323 9.4046>, rotation=<rot -1.3083 -90 -0.0011>}]}
constraints []
